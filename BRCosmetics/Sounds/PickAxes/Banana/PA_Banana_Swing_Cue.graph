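# Sound cue for the swing of the Banana cosmetic pickaxe.
# A quality-level node picks the branch, each mixing a close-attenuated banana swing with a quieter, medium-attenuated distant fire axe swing.
# Random variation shrinks down the tiers, with the last two sharing a mix of one wave per layer, while modulators vary pitch and volume.
# A concurrency override limits copies per owner, stopping the oldest.
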SoundCue /BRCosmetics/Sounds/PickAxes/Banana/PA_Banana_Swing_Cue
root=QualityLevel_1
nodes [
 {id=Attenuation_10, children [Modulator_11], properties {attenuationSettings=/Game/Sounds/Attenuation/Fort_Weapon_Pickaxe_Close_Att}}
 {id=Attenuation_11, children [WavePlayer_2], properties {attenuationSettings=/Game/Sounds/Attenuation/Fort_Weapon_Pickaxe_Medium_Att}}
 {id=Attenuation_6, children [Random_6], properties {attenuationSettings=&Fort_Weapon_Pickaxe_Close_Att}}
 {id=Attenuation_7, children [Modulator_8], properties {attenuationSettings=&Fort_Weapon_Pickaxe_Medium_Att}}
 {id=Attenuation_8, children [Random_7], properties {attenuationSettings=&Fort_Weapon_Pickaxe_Close_Att}}
 {id=Attenuation_9, children [Random_5], properties {attenuationSettings=&Fort_Weapon_Pickaxe_Medium_Att}}
 {id=Mixer_3, children [Attenuation_6, Attenuation_7], properties {inputVolume [1.1,1]}}
 {id=Mixer_4, children [Attenuation_8, Modulator_9], properties {inputVolume [1.1,1]}}
 {id=Mixer_5, children [Attenuation_10, Modulator_10], properties {inputVolume [1,1]}}
 {id=Modulator_10, children [Attenuation_11], properties {pitchMax=1.1, pitchMin=1, volumeMax=0.35, volumeMin=0.35}}
 {id=Modulator_11, children [WavePlayer_44], properties {pitchMax=1.1, pitchMin=0.9, volumeMin=0.9}}
 {id=Modulator_4, children [Mixer_3], properties {pitchMax=1.1, pitchMin=0.97, volumeMax=1, volumeMin=0.9}}
 {id=Modulator_7, children [Mixer_4]}
 {id=Modulator_8, children [Random_0], properties {pitchMax=1.1, pitchMin=1, volumeMax=0.35, volumeMin=0.35}}
 {id=Modulator_9, children [Attenuation_9], properties {pitchMax=1.1, pitchMin=1, volumeMax=0.35, volumeMin=0.35}}
 {id=QualityLevel_1, children [Modulator_4,Modulator_7,Mixer_5,Mixer_5]}
 {
  id=Random_0
  children [WavePlayer_4, WavePlayer_5, WavePlayer_12, WavePlayer_13, WavePlayer_15, WavePlayer_16]
  properties {weights [1,1,1,1,1,1]}
 }
 {
  id=Random_5
  children [WavePlayer_17, WavePlayer_18, WavePlayer_19, WavePlayer_20, WavePlayer_23, WavePlayer_26]
  properties {weights [1,1,1,1,1,1]}
 }
 {id=Random_6, children [WavePlayer_27, WavePlayer_30, WavePlayer_31, WavePlayer_38, WavePlayer_40], properties {weights [1,1,1,1,1]}}
 {id=Random_7, children [WavePlayer_41, WavePlayer_42, WavePlayer_43], properties {weights [1,1,1]}}
 {id=WavePlayer_12, properties {soundWaveAssetPtr=/Game/Athena/Sounds/Weapons/Fire_Axe_Swing_Distant_02}}
 {id=WavePlayer_13, properties {soundWaveAssetPtr=/Game/Athena/Sounds/Weapons/Fire_Axe_Swing_Distant_03}}
 {id=WavePlayer_15, properties {soundWaveAssetPtr=/Game/Athena/Sounds/Weapons/Fire_Axe_Swing_Distant_05}}
 {id=WavePlayer_16, properties {soundWaveAssetPtr=/Game/Athena/Sounds/Weapons/Fire_Axe_Swing_Distant_06}}
 {id=WavePlayer_17, properties {soundWaveAssetPtr=/Game/Athena/Sounds/Weapons/Fire_Axe_Swing_Distant_04}}
 {id=WavePlayer_18, properties {soundWaveAssetPtr=/Game/Athena/Sounds/Weapons/Fire_Axe_Swing_Distant_01}}
 {id=WavePlayer_19, properties {soundWaveAssetPtr=&Fire_Axe_Swing_Distant_02}}
 {id=WavePlayer_2, properties {soundWaveAssetPtr=&Fire_Axe_Swing_Distant_01}}
 {id=WavePlayer_20, properties {soundWaveAssetPtr=&Fire_Axe_Swing_Distant_03}}
 {id=WavePlayer_23, properties {soundWaveAssetPtr=&Fire_Axe_Swing_Distant_05}}
 {id=WavePlayer_26, properties {soundWaveAssetPtr=&Fire_Axe_Swing_Distant_06}}
 {id=WavePlayer_27, properties {soundWaveAssetPtr=&/Pickaxe_Banana_Swing_01}}
 {id=WavePlayer_30, properties {soundWaveAssetPtr=&/Pickaxe_Banana_Swing_02}}
 {id=WavePlayer_31, properties {soundWaveAssetPtr=&/Pickaxe_Banana_Swing_03}}
 {id=WavePlayer_38, properties {soundWaveAssetPtr=&/Pickaxe_Banana_Swing_04}}
 {id=WavePlayer_4, properties {soundWaveAssetPtr=&Fire_Axe_Swing_Distant_04}}
 {id=WavePlayer_40, properties {soundWaveAssetPtr=&/Pickaxe_Banana_Swing_05}}
 {id=WavePlayer_41, properties {soundWaveAssetPtr=&Pickaxe_Banana_Swing_01}}
 {id=WavePlayer_42, properties {soundWaveAssetPtr=&Pickaxe_Banana_Swing_02}}
 {id=WavePlayer_43, properties {soundWaveAssetPtr=&Pickaxe_Banana_Swing_03}}
 {id=WavePlayer_44, properties {soundWaveAssetPtr=&Pickaxe_Banana_Swing_01}}
 {id=WavePlayer_5, properties {soundWaveAssetPtr=&Fire_Axe_Swing_Distant_01}}
]
properties {
 audioPropertiesSheet=/BRCosmetics/Sounds/Settings/PropertySheets/Pickaxe/APS_Cosmetic_Pickaxe_Swing
 bOverrideConcurrency=true
 bPrimeOnLoad=true
 maxDistance=3000
 soundClassObject=&Fort_WeaponMelee_PSM
 volumeMultiplier=0.4
 concurrencyOverrides {bLimitToOwner=true, maxCount=2, resolutionRule=StopOldest}
 concurrencySet [&Player_Melee_Weapon_Conc]
}
